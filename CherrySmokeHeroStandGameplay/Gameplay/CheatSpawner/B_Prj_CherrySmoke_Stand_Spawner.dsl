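; Projectile Blueprint that spawns a random Stand building actor (BGA) on impact, overriding SpawnBGA_.
; The first impact disables collision, sets lifespan and visibility and fires an explode GameplayCue; later impacts above the spawned actor destroy it.
; Placement uses a water check, Z offset and optional normal-aligned rotation, with the normal offset multiplied by zero.
; The new actor is initialized, given the owner weapon's wrap and reported to the owner via gameplay event.
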(event EventBeginPlay)

(event Collision|EventActorBeginOverlap
  (params (OtherActor :type Actor%20Object%20Reference)))

(event EventTick
  (params (DeltaSeconds :type Float%20%28single-precision%29)))

(fn SpawnBGA_
  (params (Normal :type Vector :default "0, 0, 0") (HitActor :type Actor%20Object%20Reference) (HitComp :type Primitive%20Component%20Object%20Reference) (Pos :type Vector :default "0, 0, 0") (FromWater :type Boolean) (Hit :type Hit%20Result%20Structure))
  (bind (_returnvalue_x _returnvalue_y _returnvalue_z) (Math|Vector|BreakVector (Transformation|GetActorLocation (Variables|Spawning|GetSpawnedBGA))))
  (if (Variables|Spawning|GetSpawned)
    (if (> (.z Pos) _returnvalue_z)
      (Actor|DestroyActor (Variables|Spawning|GetSpawnedBGA)))
    (else
      (Variables|Spawning|SetSpawned true)
      (Utilities|FlowControl|Sequence
        (:then_0
          (Actor|SetLifeSpan (Variables|Spawning|GetLifespanAfterSpawnBGA))
          (Collision|SetActorEnableCollision false)
          (Rendering|SetActorHiddenInGame (Variables|Spawning|GetSetActorHiddenOnSpawnBGA))
          (GameplayCue|ExecuteGameplayCue (Game|GetPlayerPawn 0) (Variables|Default|GetImpactExplodeGC)))
        (:then_1
          _))))
  (bind _returnvalue (Building|SpawnBuildingActor (Utilities|Array|RandomArrayItem (Variables|Default|GetStand%20BGAs)) (Math|Transform|MakeTransform (Math|Vector|vector+vector (CallFunction|CheckforWater Pos HitActor) (* "0, 0, 0" (Variables|Spawning|TransformData|GetSpawnPosOffsetAlongNormal)) (Math|Vector|MakeVector 0.0 0.0 (Variables|Spawning|TransformData|GetSpawnPosAdditionalZOffset))) (Math|Rotator|SelectRotator (Math|Rotator|MakeRotfromXZ (Math|Vector|ProjectVectorontoPlane (Transformation|GetActorForwardVector (Game|GetInstigator)) (Math|Vector|SelectVector Normal "0, 0,1.000000" (Math|Vector|NotEqual(Vector) Normal "0, 0, 0" 0.0001))) (Math|Vector|SelectVector Normal "0, 0,1.000000" (Math|Vector|NotEqual(Vector) Normal "0, 0, 0" 0.0001))) (Transformation|GetActorRotation self) (Variables|Spawning|TransformData|GetUseImpactNormalForBGASpawnRotation)) (Variables|Spawning|TransformData|GetSpawnScale)) (Variables|Actor|GetInstigator) (Variables|Actor|GetInstigator)))
  (Variables|Spawning|SetSpawnedBGA _returnvalue)
  (bind _asbuilding_gameplay_actor (Utilities|Casting|CastToBuildingGameplayActor _returnvalue)
    (:then
      (Building|InitializeKismetSpawnedBuildingActor _asbuilding_gameplay_actor 0 0 true 0 false)
      (Utilities|FlowControl|Sequence
        (:then_0
          (bind _asbga_athena_with_gravity_parent (Utilities|Casting|CastToBGA_Athena_WithGravity_Parent _asbuilding_gameplay_actor)))
        (:then_1
          (bind _returnvalue_2 (CallFunction|GetOwnerWeapon))
          (CosmeticWraps|AttemptApplyOverrideWraptoActor _returnvalue_2 _asbuilding_gameplay_actor))
        (:then_2
          _)))
    (:CastFailed
      _))
  (Ability|SendGameplayEventtoActor (Actor|GetOwner) "(TagName=\"Event.Projectile.SpawnedBGA\")" (Utilities|Struct|MakeGameplayEventData :OptionalObject _returnvalue :ContextHandle "()" :InstigatorTags "(GameplayTags=)" :TargetTags "(GameplayTags=)" :EventMagnitude 0.0 :TargetData "()"))
  (return _returnvalue))

(fn ConstructionScript
  (|CallParentFunction))
